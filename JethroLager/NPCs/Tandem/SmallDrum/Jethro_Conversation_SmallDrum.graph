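# Conversation database for the SmallDrum NPC in JethroLager, providing the root dialogue entry when a player talks to the character.
# The entry point leads to a speech node with a default greeting, overridden by prioritized contextual lines for a first-ever conversation or specific player outfits.
# That node branches through link nodes to shared quest and Jethro service conversations, SmallDrum's own dialogue, and a Season 35 "Found" entry.
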ConversationDatabase /JethroLager/NPCs/Tandem/SmallDrum/Jethro_Conversation_SmallDrum
entry_points [{tag=Jethro.Conversation.Entry.SmallDrum, destinations [#1]}]
nodes [
 ConversationEntryPointNode_7 {
  id=#1
  class=ConversationEntryPointNode
  outputs [#2]
  properties {entryTag {tagName=Jethro.Conversation.Entry.SmallDrum}}
 }
 Speech_C_4 {
  id=#2
  class=Speech_C
  outputs [#3, #4, #5, #6]
  properties {
   nodeName=""
   generalConfig {
    defaultMessage=What've you heard?
    contextualMessages [
     {
      message=Have I met you before? You seem... familiar.
      contextRequirements [{requirement {class=FortControllerRequirement_IsFirstEverConversationWithNPC}}]
      priority {curve {rowName=Tandem.Speech.Priority.FirstEverInteraction}}
     }
     {
      message="You look way too pleased with yourself, Han..."
      contextRequirements [{requirement {class=FortControllerRequirement_HasCID, properties {allowedCIds [/BRCosmetics/Athena/Items/Cosmetics/Characters/Character_DistantEchoPilot]}}}]
      priority {curve {rowName=Tandem.Speech.Priority.OutfitReaction}}
     }
     {
      message=You were right to trust these people. How can I help?
      contextRequirements [{requirement {class=FortControllerRequirement_HasCID, properties {allowedCIds [/BRCosmetics/Athena/Items/Cosmetics/Characters/Character_DistantEchoPro]}}}]
      priority {curve {rowName=Tandem.Speech.Priority.OutfitReaction}}
     }
     {
      message="If you play your part, we're more than happy to have you."
      contextRequirements [{requirement {
         class=FortControllerRequirement_HasCID
         properties {allowedCIds [&Character_UndergroundRebel_Fashion]}
        }}]
      priority {curve {rowName=Tandem.Speech.Priority.OutfitReaction}}
     }
    ]
   }
  }
 }
 ConversationLinkNode_2 {
  id=#6
  class=ConversationLinkNode
  properties {remoteEntryTag {tagName=Conversation.Entry.Season35.Found.SmallDrum.01}}
 }
 ConversationLinkNode_3 {
  id=#5
  class=ConversationLinkNode
  properties {remoteEntryTag {tagName=Jethro.Conversation.Entry.SmallDrum.Dialogue}}
 }
 ConversationLinkNode_0 {id=#4, class=ConversationLinkNode, properties {remoteEntryTag {tagName=Conversation.Entry.JethroServices}}}
 ConversationLinkNode_1 {id=#3, class=ConversationLinkNode, properties {remoteEntryTag {tagName=Conversation.Entry.NPCLibrary.Quests}}}
]
