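# Conversation database for the SmallDrum NPC's dialogue, a question menu about the Empire and the rebellion.
# A tagged entry point opens a "What's up?" speech node that branches into four topics and a Back option, labeled by choice sub-nodes.
# Three topics give a single reply, while "Joining the cause" chains three lines.
# All topic branches converge on a shared ReturnToStart node labeled BACK.
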ConversationDatabase /JethroLager/NPCs/Tandem/SmallDrum/Jethro_Conversation_SmallDrum_Dialogue
entry_points [{tag=Jethro.Conversation.Entry.SmallDrum.Dialogue, destinations [#1]}]
nodes [
 ConversationEntryPointNode_0 {
  id=#1
  class=ConversationEntryPointNode
  outputs [#2]
  properties {entryTag {tagName=Jethro.Conversation.Entry.SmallDrum.Dialogue}}
 }
 Speech_C_0 {
  id=#2
  class=Speech_C
  outputs [
   #3
   #4
   #5
   #6
   #7
  ]
  properties {nodeName="", softTaskIcon=&T_UI_NPC_question, generalConfig {defaultMessage=What's up?}}
  sub_nodes [ConversationChoiceNode_5 {class=ConversationChoiceNode, properties {defaultChoiceDisplayText=Ask about...}}]
 }
 Speech_C_3 {
  id=#7
  class=Speech_C
  outputs [#8]
  properties {
   nodeName=""
   softTaskIcon=&T_UI_NPC_question
   generalConfig {defaultMessage=You'll find no finer training than what the rebellion offers.}
  }
  sub_nodes [ConversationChoiceNode_3 {class=ConversationChoiceNode, properties {defaultChoiceDisplayText=Joining the cause}}]
 }
 Speech_C_4 {
  id=#8
  class=Speech_C
  outputs [#9]
  properties {
   nodeName=""
   softTaskIcon=&T_UI_NPC_question
   generalConfig {defaultMessage=Find them. Become worthy to this great cause.}
  }
 }
 Speech_C_5 {
  id=#9
  class=Speech_C
  outputs [#10]
  properties {
   nodeName=""
   softTaskIcon=&T_UI_NPC_question
   generalConfig {defaultMessage=The rebellion never forgets a friend.}
  }
 }
 ReturnToStart_C_23 {
  id=#10
  class=ReturnToStart_C
  properties {softTaskIcon=&T_UI_NPC_Back}
  sub_nodes [ConversationChoiceNode_32 {class=ConversationChoiceNode, properties {defaultChoiceDisplayText=BACK}}]
 }
 Speech_C_2 {
  id=#6
  class=Speech_C
  outputs [#10]
  properties {
   nodeName=""
   softTaskIcon=&T_UI_NPC_question
   generalConfig {defaultMessage=Some secrets are best kept secret. For now.}
  }
  sub_nodes [ConversationChoiceNode_2 {class=ConversationChoiceNode, properties {defaultChoiceDisplayText=Empire's Plans}}]
 }
 Speech_C_1 {
  id=#5
  class=Speech_C
  outputs [#10]
  properties {
   nodeName=""
   softTaskIcon=&T_UI_NPC_question
   generalConfig {defaultMessage=We're always looking for trustworthy friends.}
  }
  sub_nodes [ConversationChoiceNode_1 {class=ConversationChoiceNode, properties {defaultChoiceDisplayText=The Rebels}}]
 }
 Back_C_0 {
  id=#4
  class=Back_C
  sub_nodes [ConversationChoiceNode_7 {class=ConversationChoiceNode, properties {defaultChoiceDisplayText=BACK}}]
 }
 Speech_C_7 {
  id=#3
  class=Speech_C
  outputs [#10]
  properties {
   nodeName=""
   softTaskIcon=&T_UI_NPC_question
   generalConfig {defaultMessage="That's a big quest. I'll share what I know when the time is right. "}
  }
  sub_nodes [ConversationChoiceNode_0 {class=ConversationChoiceNode, properties {defaultChoiceDisplayText=The Empire}}]
 }
]
